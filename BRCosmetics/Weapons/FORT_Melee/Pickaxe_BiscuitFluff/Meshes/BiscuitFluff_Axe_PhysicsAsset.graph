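# Physics asset for the BiscuitFluff two-handed pickaxe, giving collision bodies and joints to its handle and a dangling cake piece.
# Four capsules approximate the handle, while base_cake, dyn_cake_01 and dyn_cake_02 form a chain of single-capsule bodies ending in the largest.
# base_cake rotates freely on all axes against the handle.
# The two joints further along lock twist and limit swing, with a looser range at the tip.
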PhysicsAsset /BRCosmetics/Weapons/FORT_Melee/Pickaxe_BiscuitFluff/Meshes/BiscuitFluff_Axe_PhysicsAsset
bodies {
 Melee_TwoHanded_Handle [
  {shapeType=Capsule, length=11.1176, radius=1.5628, shapeName=Melee_TwoHanded_Handle_capsule, center=<0.7299 -11.8385 93.0417>}
  {shapeType=Capsule, length=94.5826, radius=2.077, shapeName=Melee_TwoHanded_Handle_capsule, center=<0 -0.6506 10.4486>}
  {
   shapeType=Capsule
   length=7.429
   radius=1.6433
   shapeName=Melee_TwoHanded_Handle_capsule
   center=<0.7179 -4.0184 87.5293>
   rotation=<rot 0 0 -120.0004>
  }
  {
   shapeType=Capsule
   length=15.1773
   radius=1.6441
   shapeName=Melee_TwoHanded_Handle_capsule
   center=<0.7519 8.1212 84.6306>
   rotation=<rot 0 0 -60.0002>
  }
 ]
 base_cake [{shapeType=Capsule, length=0, radius=1.749, shapeName=base_cake_capsule, center=<0 0.431 0>, rotation=<rot 90 180 -89.9997>}]
 dyn_cake_01 [{shapeType=Capsule, length=4.1312, radius=2.049, shapeName=dyn_cake_01_capsule, center=<4.304 0 0>}]
 dyn_cake_02 [{shapeType=Capsule, length=4.2336, radius=6.4688, shapeName=dyn_cake_02_capsule, center=<14.4338 -0.2353 0>, rotation=<rot 0 -5 90>}]
}
constraints [
 {bone1Name=dyn_cake_01, bone2Name=base_cake, swing1LimitDegrees=5, swing2LimitDegrees=30, twistLimitDegrees=45, twistMotion=Locked}
 {
  bone1Name=base_cake
  bone2Name=Melee_TwoHanded_Handle
  swing1LimitDegrees=45
  swing1Motion=Free
  swing2LimitDegrees=45
  swing2Motion=Free
  twistLimitDegrees=45
  twistMotion=Free
 }
 {bone1Name=dyn_cake_02, bone2Name=dyn_cake_01, swing1LimitDegrees=50, swing2LimitDegrees=40, twistLimitDegrees=45, twistMotion=Locked}
]
